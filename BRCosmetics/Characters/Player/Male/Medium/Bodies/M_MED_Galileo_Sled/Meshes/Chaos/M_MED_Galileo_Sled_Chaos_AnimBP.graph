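# Post-processes the M_MED_Galileo_Sled player body, correcting armor, strap, pelvis and belt bones over a pose copied from the attached parent mesh.
# RBF pose drivers on both upper arms drive shoulder bones, then constraints and look-ats position shoulder and chest armor.
# Additive bone offsets blend by clavicle and thigh rotations, measured from ref pose in EventGraph, plus crouch states.
# Only the right shoulder look-at depends on skydiving; the left is constant.
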AnimBlueprint /BRCosmetics/Characters/Player/Male/Medium/Bodies/M_MED_Galileo_Sled/Meshes/Chaos/M_MED_Galileo_Sled_Chaos_AnimBP
target_skeleton=/Game/Characters/Player/Male/Medium/Bodies/M_MED_Galileo_Sled/Meshes/M_MED_Galileo_Sled_Skeleton
graphs [
 AnimGraph {nodes [
   n1 = Get thigh rotation_r() -> (thigh rotation_r:double)
   n2 = Pose Driver - Source: upperarm_l!(SourcePose:Pose Link Structure=Copy Pose From Mesh() {bUseAttachedParent=true}) -> (Pose:Pose Link Structure) {
    is=title AnimGraphNode_PoseDriver
    poseAsset=/BRCosmetics/Characters/Player/Male/Medium/Bodies/M_MED_Galileo_Sled/Meshes/M_MED_Galileo_Sled_Shldrs_RBF_New_PoseAsset
    onlyDriveBones [{boneName=dyn_shoulder_armor_l_RBF}, {boneName=dyn_shoulder_strap_l}]
    poseTargets [
     {
      bApplyCustomCurve=false
      bIsHidden=false
      distanceMethod=DefaultMethod
      drivenName=Pose_27
      functionType=DefaultFunction
      targetScale=0.9926
      boneTransforms [{targetRotation=<rot -37.7574 0.1509 2.6875>, targetTranslation=<15.5111 0 0>}]
      customCurve {defaultValue=3.4028234663852886e+38, postInfinityExtrap=RCCE_Constant, preInfinityExtrap=RCCE_Constant, keys []}
      targetRotation=<rot 0 0 0>
     }
     {
      bApplyCustomCurve=false
      bIsHidden=false
      distanceMethod=DefaultMethod
      drivenName=Pose_28
      functionType=DefaultFunction
      targetScale=1
      boneTransforms [{targetRotation=<rot 7.6449 2.0825 2.1436>, targetTranslation=<15.5111 0 0>}]
      customCurve {defaultValue=3.4028234663852886e+38, postInfinityExtrap=RCCE_Constant, preInfinityExtrap=RCCE_Constant, keys []}
      targetRotation=<rot 0 0 0>
     }
     {
      bApplyCustomCurve=false
      bIsHidden=false
      distanceMethod=DefaultMethod
      drivenName=Pose_29
      functionType=DefaultFunction
      targetScale=0.9739
      boneTransforms [{targetRotation=<rot 57.226 5.1005 3.9268>, targetTranslation=<15.5111 0 0>}]
      customCurve {defaultValue=3.4028234663852886e+38, postInfinityExtrap=RCCE_Constant, preInfinityExtrap=RCCE_Constant, keys []}
      targetRotation=<rot 0 0 0>
     }
     {
      bApplyCustomCurve=false
      bIsHidden=false
      distanceMethod=DefaultMethod
      drivenName=Pose_30
      functionType=DefaultFunction
      targetScale=0.5974
      boneTransforms [{targetRotation=<rot 28.0414 52.3966 7.0808>, targetTranslation=<15.5111 0 0>}]
      customCurve {defaultValue=3.4028234663852886e+38, postInfinityExtrap=RCCE_Constant, preInfinityExtrap=RCCE_Constant, keys []}
      targetRotation=<rot 0 0 0>
     }
     {
      bApplyCustomCurve=false
      bIsHidden=false
      distanceMethod=DefaultMethod
      drivenName=Pose_31
      functionType=DefaultFunction
      targetScale=0.709
      boneTransforms [{targetRotation=<rot -2.3017 93.1218 7.599>, targetTranslation=<15.5111 0 0>}]
      customCurve {defaultValue=3.4028234663852886e+38, postInfinityExtrap=RCCE_Constant, preInfinityExtrap=RCCE_Constant, keys []}
      targetRotation=<rot 0 0 0>
     }
     {
      bApplyCustomCurve=false
      bIsHidden=false
      distanceMethod=DefaultMethod
      drivenName=Pose_32
      functionType=DefaultFunction
      targetScale=0.58
      boneTransforms [{targetRotation=<rot 3.654 49.1104 7.0518>, targetTranslation=<15.5111 0 0>}]
      customCurve {defaultValue=3.4028234663852886e+38, postInfinityExtrap=RCCE_Constant, preInfinityExtrap=RCCE_Constant, keys []}
      targetRotation=<rot 0 0 0>
     }
     {
      bApplyCustomCurve=false
      bIsHidden=false
      distanceMethod=DefaultMethod
      drivenName=Pose_33
      functionType=DefaultFunction
      targetScale=0.7046
      boneTransforms [{targetRotation=<rot 5.719 -33.0659 -2.0447>, targetTranslation=<15.5111 0 0>}]
      customCurve {defaultValue=3.4028234663852886e+38, postInfinityExtrap=RCCE_Constant, preInfinityExtrap=RCCE_Constant, keys []}
      targetRotation=<rot 0 0 0>
     }
     {
      bApplyCustomCurve=false
      bIsHidden=false
      distanceMethod=DefaultMethod
      drivenName=Pose_34
      functionType=DefaultFunction
      targetScale=0.7046
      boneTransforms [{targetRotation=<rot 5.272 -62.1971 -5.9428>, targetTranslation=<15.5111 0 0>}]
      customCurve {defaultValue=3.4028234663852886e+38, postInfinityExtrap=RCCE_Constant, preInfinityExtrap=RCCE_Constant, keys []}
      targetRotation=<rot 0 0 0>
     }
     {
      bApplyCustomCurve=false
      bIsHidden=false
      distanceMethod=DefaultMethod
      drivenName=Pose_35
      functionType=DefaultFunction
      targetScale=0.6884
      boneTransforms [{targetRotation=<rot -22.8671 -46.383 9.8347>, targetTranslation=<15.5111 0 0>}]
      customCurve {defaultValue=3.4028234663852886e+38, postInfinityExtrap=RCCE_Constant, preInfinityExtrap=RCCE_Constant, keys []}
      targetRotation=<rot 0 0 0>
     }
     {
      bApplyCustomCurve=false
      bIsHidden=false
      distanceMethod=DefaultMethod
      drivenName=Pose_36
      functionType=DefaultFunction
      targetScale=0.7716
      boneTransforms [{targetRotation=<rot -78.4439 -8.6556 10.6642>, targetTranslation=<15.5111 0 0>}]
      customCurve {defaultValue=3.4028234663852886e+38, postInfinityExtrap=RCCE_Constant, preInfinityExtrap=RCCE_Constant, keys []}
      targetRotation=<rot 0 0 0>
     }
     {
      bApplyCustomCurve=false
      bIsHidden=false
      distanceMethod=DefaultMethod
      drivenName=Pose_37
      functionType=DefaultFunction
      targetScale=0.8
      boneTransforms [{targetRotation=<rot -41.935 55.3657 -40.6602>, targetTranslation=<15.5111 0 0>}]
      customCurve {defaultValue=3.4028234663852886e+38, postInfinityExtrap=RCCE_Constant, preInfinityExtrap=RCCE_Constant, keys []}
      targetRotation=<rot 0 0 0>
     }
     {
      bApplyCustomCurve=false
      bIsHidden=false
      distanceMethod=DefaultMethod
      drivenName=Pose_38
      functionType=DefaultFunction
      targetScale=0.6884
      boneTransforms [{targetRotation=<rot -43.4624 -69.9666 69.4913>, targetTranslation=<15.5111 0 0>}]
      customCurve {defaultValue=3.4028234663852886e+38, postInfinityExtrap=RCCE_Constant, preInfinityExtrap=RCCE_Constant, keys []}
      targetRotation=<rot 0 0 0>
     }
     {
      bApplyCustomCurve=false
      bIsHidden=false
      distanceMethod=DefaultMethod
      drivenName=Pose_39
      functionType=DefaultFunction
      targetScale=0.6159
      boneTransforms [{targetRotation=<rot 48.0922 72.4313 -6.6698>, targetTranslation=<15.5111 0 0>}]
      customCurve {defaultValue=3.4028234663852886e+38, postInfinityExtrap=RCCE_Constant, preInfinityExtrap=RCCE_Constant, keys []}
      targetRotation=<rot 0 0 0>
     }
     {
      bApplyCustomCurve=false
      bIsHidden=false
      distanceMethod=DefaultMethod
      drivenName=Pose_40
      functionType=DefaultFunction
      targetScale=0.7716
      boneTransforms [{targetRotation=<rot -69.4979 -172.5092 173.9242>, targetTranslation=<15.5111 0 0>}]
      customCurve {defaultValue=3.4028234663852886e+38, postInfinityExtrap=RCCE_Constant, preInfinityExtrap=RCCE_Constant, keys []}
      targetRotation=<rot 0 0 0>
     }
     {
      bApplyCustomCurve=false
      bIsHidden=false
      distanceMethod=DefaultMethod
      drivenName=Pose_41
      functionType=DefaultFunction
      targetScale=0.58
      boneTransforms [{targetRotation=<rot -13.4769 65.8568 -31.4924>, targetTranslation=<15.5111 0 0>}]
      customCurve {defaultValue=3.4028234663852886e+38, postInfinityExtrap=RCCE_Constant, preInfinityExtrap=RCCE_Constant, keys []}
      targetRotation=<rot 0 0 0>
     }
     {
      bApplyCustomCurve=false
      bIsHidden=false
      distanceMethod=DefaultMethod
      drivenName=Pose_42
      functionType=DefaultFunction
      targetScale=0.9065
      boneTransforms [{targetRotation=<rot -53.1476 109.6244 -57.8883>, targetTranslation=<15.5111 0 0>}]
      customCurve {defaultValue=3.4028234663852886e+38, postInfinityExtrap=RCCE_Constant, preInfinityExtrap=RCCE_Constant, keys []}
      targetRotation=<rot 0 0 0>
     }
     {
      bApplyCustomCurve=false
      bIsHidden=false
      distanceMethod=DefaultMethod
      drivenName=Pose_43
      functionType=DefaultFunction
      targetScale=0.921
      boneTransforms [{targetRotation=<rot 32.4961 118.7317 28.9677>, targetTranslation=<15.5111 0 0>}]
      customCurve {defaultValue=3.4028234663852886e+38, postInfinityExtrap=RCCE_Constant, preInfinityExtrap=RCCE_Constant, keys []}
      targetRotation=<rot 0 0 0>
     }
     {
      bApplyCustomCurve=false
      bIsHidden=false
      distanceMethod=DefaultMethod
      drivenName=Pose_44
      functionType=DefaultFunction
      targetScale=0.835
      boneTransforms [{targetRotation=<rot 39.4008 -57.7573 -45.6614>, targetTranslation=<15.5111 0 0>}]
      customCurve {defaultValue=3.4028234663852886e+38, postInfinityExtrap=RCCE_Constant, preInfinityExtrap=RCCE_Constant, keys []}
      targetRotation=<rot 0 0 0>
     }
     {
      bApplyCustomCurve=false
      bIsHidden=false
      distanceMethod=DefaultMethod
      drivenName=Pose_45
      functionType=DefaultFunction
      targetScale=1
      boneTransforms [{targetRotation=<rot -6.7303 134.1828 4.2186>, targetTranslation=<15.5111 0 0>}]
      customCurve {defaultValue=3.4028234663852886e+38, postInfinityExtrap=RCCE_Constant, preInfinityExtrap=RCCE_Constant, keys []}
      targetRotation=<rot 0 0 0>
     }
    ]
    rBFParams {radius=35}
    sourceBones [{boneName=upperarm_l}]
   }
   n3 = Pose Driver - Source: upperarm_r!(SourcePose:Pose Link Structure=@n2) -> (Pose:Pose Link Structure) {
    is=title AnimGraphNode_PoseDriver
    axisLength=-20
    poseAsset=&M_MED_Galileo_Sled_Shldrs_RBF_New_PoseAsset
    onlyDriveBones [{boneName=dyn_shoulder_armor_r_RBF}, {boneName=dyn_shoulder_strap_r}]
    poseTargets [
     {
      bApplyCustomCurve=false
      bIsHidden=false
      distanceMethod=DefaultMethod
      drivenName=Pose_27
      functionType=DefaultFunction
      targetScale=0.9926
      boneTransforms [{targetRotation=<rot -37.7574 0.1509 2.6875>, targetTranslation=<-15.5111 0 0>}]
      customCurve {defaultValue=3.4028234663852886e+38, postInfinityExtrap=RCCE_Constant, preInfinityExtrap=RCCE_Constant, keys []}
      targetRotation=<rot 0 0 0>
     }
     {
      bApplyCustomCurve=false
      bIsHidden=false
      distanceMethod=DefaultMethod
      drivenName=Pose_28
      functionType=DefaultFunction
      targetScale=0.8493
      boneTransforms [{targetRotation=<rot 7.6449 2.0825 2.1436>, targetTranslation=<-15.5111 0 0>}]
      customCurve {defaultValue=3.4028234663852886e+38, postInfinityExtrap=RCCE_Constant, preInfinityExtrap=RCCE_Constant, keys []}
      targetRotation=<rot 0 0 0>
     }
     {
      bApplyCustomCurve=false
      bIsHidden=false
      distanceMethod=DefaultMethod
      drivenName=Pose_29
      functionType=DefaultFunction
      targetScale=0.9739
      boneTransforms [{targetRotation=<rot 57.226 5.1005 3.9268>, targetTranslation=<-15.5111 0 0>}]
      customCurve {defaultValue=3.4028234663852886e+38, postInfinityExtrap=RCCE_Constant, preInfinityExtrap=RCCE_Constant, keys []}
      targetRotation=<rot 0 0 0>
     }
     {
      bApplyCustomCurve=false
      bIsHidden=false
      distanceMethod=DefaultMethod
      drivenName=Pose_30
      functionType=DefaultFunction
      targetScale=0.5974
      boneTransforms [{targetRotation=<rot 28.0414 52.3966 7.0808>, targetTranslation=<-15.5111 0 0>}]
      customCurve {defaultValue=3.4028234663852886e+38, postInfinityExtrap=RCCE_Constant, preInfinityExtrap=RCCE_Constant, keys []}
      targetRotation=<rot 0 0 0>
     }
     {
      bApplyCustomCurve=false
      bIsHidden=false
      distanceMethod=DefaultMethod
      drivenName=Pose_31
      functionType=DefaultFunction
      targetScale=0.709
      boneTransforms [{targetRotation=<rot -2.3017 93.1218 7.599>, targetTranslation=<-15.5111 0 0>}]
      customCurve {defaultValue=3.4028234663852886e+38, postInfinityExtrap=RCCE_Constant, preInfinityExtrap=RCCE_Constant, keys []}
      targetRotation=<rot 0 0 0>
     }
     {
      bApplyCustomCurve=false
      bIsHidden=false
      distanceMethod=DefaultMethod
      drivenName=Pose_32
      functionType=DefaultFunction
      targetScale=0.58
      boneTransforms [{targetRotation=<rot 3.654 49.1104 7.0518>, targetTranslation=<-15.5111 0 0>}]
      customCurve {defaultValue=3.4028234663852886e+38, postInfinityExtrap=RCCE_Constant, preInfinityExtrap=RCCE_Constant, keys []}
      targetRotation=<rot 0 0 0>
     }
     {
      bApplyCustomCurve=false
      bIsHidden=false
      distanceMethod=DefaultMethod
      drivenName=Pose_33
      functionType=DefaultFunction
      targetScale=0.7046
      boneTransforms [{targetRotation=<rot 5.719 -33.0659 -2.0447>, targetTranslation=<-15.5111 0 0>}]
      customCurve {defaultValue=3.4028234663852886e+38, postInfinityExtrap=RCCE_Constant, preInfinityExtrap=RCCE_Constant, keys []}
      targetRotation=<rot 0 0 0>
     }
     {
      bApplyCustomCurve=false
      bIsHidden=false
      distanceMethod=DefaultMethod
      drivenName=Pose_34
      functionType=DefaultFunction
      targetScale=0.7046
      boneTransforms [{targetRotation=<rot 5.272 -62.1971 -5.9428>, targetTranslation=<-15.5111 0 0>}]
      customCurve {defaultValue=3.4028234663852886e+38, postInfinityExtrap=RCCE_Constant, preInfinityExtrap=RCCE_Constant, keys []}
      targetRotation=<rot 0 0 0>
     }
     {
      bApplyCustomCurve=false
      bIsHidden=false
      distanceMethod=DefaultMethod
      drivenName=Pose_35
      functionType=DefaultFunction
      targetScale=0.6884
      boneTransforms [{targetRotation=<rot -22.8671 -46.383 9.8347>, targetTranslation=<-15.5111 0 0>}]
      customCurve {defaultValue=3.4028234663852886e+38, postInfinityExtrap=RCCE_Constant, preInfinityExtrap=RCCE_Constant, keys []}
      targetRotation=<rot 0 0 0>
     }
     {
      bApplyCustomCurve=false
      bIsHidden=false
      distanceMethod=DefaultMethod
      drivenName=Pose_36
      functionType=DefaultFunction
      targetScale=0.7716
      boneTransforms [{targetRotation=<rot -78.4439 -8.6556 10.6642>, targetTranslation=<-15.5111 0 0>}]
      customCurve {defaultValue=3.4028234663852886e+38, postInfinityExtrap=RCCE_Constant, preInfinityExtrap=RCCE_Constant, keys []}
      targetRotation=<rot 0 0 0>
     }
     {
      bApplyCustomCurve=false
      bIsHidden=false
      distanceMethod=DefaultMethod
      drivenName=Pose_37
      functionType=DefaultFunction
      targetScale=0.726
      boneTransforms [{targetRotation=<rot -41.935 55.3657 -40.6602>, targetTranslation=<-15.5111 0 0>}]
      customCurve {defaultValue=3.4028234663852886e+38, postInfinityExtrap=RCCE_Constant, preInfinityExtrap=RCCE_Constant, keys []}
      targetRotation=<rot 0 0 0>
     }
     {
      bApplyCustomCurve=false
      bIsHidden=false
      distanceMethod=DefaultMethod
      drivenName=Pose_38
      functionType=DefaultFunction
      targetScale=0.6884
      boneTransforms [{targetRotation=<rot -43.4624 -69.9666 69.4913>, targetTranslation=<-15.5111 0 0>}]
      customCurve {defaultValue=3.4028234663852886e+38, postInfinityExtrap=RCCE_Constant, preInfinityExtrap=RCCE_Constant, keys []}
      targetRotation=<rot 0 0 0>
     }
     {
      bApplyCustomCurve=false
      bIsHidden=false
      distanceMethod=DefaultMethod
      drivenName=Pose_39
      functionType=DefaultFunction
      targetScale=0.6159
      boneTransforms [{targetRotation=<rot 48.0922 72.4313 -6.6698>, targetTranslation=<-15.5111 0 0>}]
      customCurve {defaultValue=3.4028234663852886e+38, postInfinityExtrap=RCCE_Constant, preInfinityExtrap=RCCE_Constant, keys []}
      targetRotation=<rot 0 0 0>
     }
     {
      bApplyCustomCurve=false
      bIsHidden=false
      distanceMethod=DefaultMethod
      drivenName=Pose_40
      functionType=DefaultFunction
      targetScale=0.7716
      boneTransforms [{targetRotation=<rot -69.4979 -172.5092 173.9242>, targetTranslation=<-15.5111 0 0>}]
      customCurve {defaultValue=3.4028234663852886e+38, postInfinityExtrap=RCCE_Constant, preInfinityExtrap=RCCE_Constant, keys []}
      targetRotation=<rot 0 0 0>
     }
     {
      bApplyCustomCurve=false
      bIsHidden=false
      distanceMethod=DefaultMethod
      drivenName=Pose_41
      functionType=DefaultFunction
      targetScale=0.58
      boneTransforms [{targetRotation=<rot -13.4769 65.8568 -31.4924>, targetTranslation=<-15.5111 0 0>}]
      customCurve {defaultValue=3.4028234663852886e+38, postInfinityExtrap=RCCE_Constant, preInfinityExtrap=RCCE_Constant, keys []}
      targetRotation=<rot 0 0 0>
     }
     {
      bApplyCustomCurve=false
      bIsHidden=false
      distanceMethod=DefaultMethod
      drivenName=Pose_42
      functionType=DefaultFunction
      targetScale=1
      boneTransforms [{targetRotation=<rot -53.1476 109.6244 -57.8883>, targetTranslation=<-15.5111 0 0>}]
      customCurve {defaultValue=3.4028234663852886e+38, postInfinityExtrap=RCCE_Constant, preInfinityExtrap=RCCE_Constant, keys []}
      targetRotation=<rot 0 0 0>
     }
     {
      bApplyCustomCurve=false
      bIsHidden=false
      distanceMethod=DefaultMethod
      drivenName=Pose_43
      functionType=DefaultFunction
      targetScale=0.921
      boneTransforms [{targetRotation=<rot 32.4961 118.7317 28.9677>, targetTranslation=<-15.5111 0 0>}]
      customCurve {defaultValue=3.4028234663852886e+38, postInfinityExtrap=RCCE_Constant, preInfinityExtrap=RCCE_Constant, keys []}
      targetRotation=<rot 0 0 0>
     }
     {
      bApplyCustomCurve=false
      bIsHidden=false
      distanceMethod=DefaultMethod
      drivenName=Pose_44
      functionType=DefaultFunction
      targetScale=0.835
      boneTransforms [{targetRotation=<rot 39.4008 -57.7573 -45.6614>, targetTranslation=<-15.5111 0 0>}]
      customCurve {defaultValue=3.4028234663852886e+38, postInfinityExtrap=RCCE_Constant, preInfinityExtrap=RCCE_Constant, keys []}
      targetRotation=<rot 0 0 0>
     }
     {
      bApplyCustomCurve=false
      bIsHidden=false
      distanceMethod=DefaultMethod
      drivenName=Pose_45
      functionType=DefaultFunction
      targetScale=0.8
      boneTransforms [{targetRotation=<rot -6.7303 134.1828 4.2186>, targetTranslation=<-15.5111 0 0>}]
      customCurve {defaultValue=3.4028234663852886e+38, postInfinityExtrap=RCCE_Constant, preInfinityExtrap=RCCE_Constant, keys []}
      targetRotation=<rot 0 0 0>
     }
    ]
    rBFParams {radius=35}
    sourceBones [{boneName=upperarm_r}]
   }
   n4 = "Constraint - dyn_shoulder_armor_l_MSTR (PPPP)"!(
    ComponentPose=Local To Component(LocalPose=@n3)
    Alpha=1
    ConstraintWeights_0=0.0856
    ConstraintWeights_1:float=0.0272
    ConstraintWeights_2:float=0.7506
    ConstraintWeights_3:float=0.1366
   ) -> (Pose) {
    alphaBoolBlend {blendOption=Linear}
    boneToModify {boneName=dyn_shoulder_armor_l_MSTR}
    constraintSetup [
     {targetBone {boneName=dyn_chest_armor_bk_end_l}}
     {targetBone {boneName=spine_05}}
     {targetBone {boneName=dyn_shoulder_strap_l}}
     {targetBone {boneName=dyn_chest_armor_fr_end_l}}
    ]
    constraintWeights [1,1,1,1]
   }
   n5 = Look At - Bone: dyn_shoulder_armor_l_Aim(ComponentPose=@n4, Alpha=0.1) {
    alphaBoolBlend {blendOption=Linear}
    boneToModify {boneName=dyn_shoulder_armor_l_Aim}
    lookAtLocation {x=0}
    lookAtTarget {boneReference {boneName=lowerarm_l}}
    lookAt_Axis {axis=<1 0>}
    lookUp_Axis {axis {x=0, z=1}}
   }
   n6 = "Constraint - dyn_shoulder_armor_r_MSTR (PPPP)"!(
    ComponentPose=@n5
    Alpha=1
    ConstraintWeights_0=0.7506
    ConstraintWeights_1:float=0.0856
    ConstraintWeights_2:float=0.1366
    ConstraintWeights_3:float=0.0272
   ) -> (Pose) {
    alphaBoolBlend {blendOption=Linear}
    boneToModify {boneName=dyn_shoulder_armor_r_MSTR}
    constraintSetup [
     {targetBone {boneName=dyn_shoulder_strap_r}}
     {targetBone {boneName=dyn_chest_armor_bk_end_r}}
     {targetBone {boneName=dyn_chest_armor_fr_end_r}}
     {targetBone {boneName=spine_05}}
    ]
    constraintWeights [1,1,1,1]
   }
   n7 = Look At - Bone: dyn_shoulder_armor_r_Aim(ComponentPose=@n6, Alpha=Get IsSkyDiving() -> (IsSkyDiving:double)) {
    alphaBoolBlend {blendOption=Linear}
    alphaScaleBiasClamp {bMapRange=true, outRange {max=0.1}}
    boneToModify {boneName=dyn_shoulder_armor_r_Aim}
    lookAtLocation {x=0}
    lookAtTarget {boneReference {boneName=lowerarm_r}}
    lookAt_Axis {axis=<1 0>}
    lookUp_Axis {axis {x=0, z=1}}
   }
   n8 = "Transform (Modify) Bone - Bone: dyn_shoulder_armor_l_MSTR"!(ComponentPose=@n7, Alpha=Get clav_l_rot() -> (clav_l_rot:double), Translation="0,0,0", Rotation="12,10,0") -> (Pose) {
    rotationMode=BMM_Additive
    rotationSpace=BCS_BoneSpace
    translationMode=BMM_Additive
    translationSpace=BCS_BoneSpace
    alphaBoolBlend {blendInTime=0.1, blendOption=Linear, blendOutTime=0.1}
    alphaScaleBiasClamp {bInterpResult=true}
    boneToModify {boneName=dyn_shoulder_armor_l_MSTR}
    rotation {pitch=12, yaw=10}
   }
   n9 = "Transform (Modify) Bone - Bone: dyn_shoulder_armor_r_MSTR"!(ComponentPose=@n8, Alpha=Get clav_r_rot() -> (clav_r_rot:double), Translation="0,0,0", Rotation="12,-10,0") -> (Pose) {
    rotationMode=BMM_Additive
    rotationSpace=BCS_BoneSpace
    translationMode=BMM_Additive
    translationSpace=BCS_BoneSpace
    alphaBoolBlend {blendInTime=0.1, blendOption=Linear, blendOutTime=0.1}
    alphaScaleBiasClamp {bInterpResult=true}
    boneToModify {boneName=dyn_shoulder_armor_r_MSTR}
    rotation {pitch=12, yaw=-10}
   }
   n10 = "Transform (Modify) Bone - Bone: dyn_shoulder_strap_l_MSTR"(ComponentPose=@n9, Alpha=Get clav_l_rot(), Translation="2.000,3.000,-0.5", Rotation="0,0,0", Scale="0.75,1.25,0.75") {
    rotationMode=BMM_Additive
    rotationSpace=BCS_BoneSpace
    scaleMode=BMM_Replace
    translationMode=BMM_Additive
    translationSpace=BCS_BoneSpace
    alphaBoolBlend {blendOption=Linear}
    alphaScaleBiasClamp {outRange {min=0.25}}
    boneToModify {boneName=dyn_shoulder_strap_l_MSTR}
    scale=<0.75 1.25 0.75>
    translation=<2 3 -0.5>
   }
   n11 = Get thigh_rot_l()
   n12 = "Transform (Modify) Bone - Bone: pelvis_l"(ComponentPose=@n10, Alpha=@n11, Translation="-2.088,-4.009,-1.461", Rotation="0,0,0", Scale="1.488,1.000,1.000") {
    scaleMode=BMM_Additive
    translationMode=BMM_Additive
    translationSpace=BCS_ParentBoneSpace
    alphaBoolBlend {blendOption=Linear}
    boneToModify {boneName=pelvis_l}
    scale {x=1.488}
    translation=<-2.088 -4.009 -1.461>
   }
   n13 = "Transform (Modify) Bone - Bone: pelvis_mid_l"(ComponentPose=@n12, Alpha=@n11, Translation="-2.366,0.140,0.263", Rotation="0,0,0", Scale="1.216,1.220,1.000") {
    scaleMode=BMM_Additive
    translationMode=BMM_Additive
    translationSpace=BCS_ParentBoneSpace
    alphaBoolBlend {blendOption=Linear}
    boneToModify {boneName=pelvis_mid_l}
    scale=<1.216 1.22>
    translation=<-2.3657 0.14 0.263>
   }
   n14 = "Transform (Modify) Bone - Bone: pelvis_side_l"(ComponentPose=@n13, Alpha=@n11, Translation="-0.228,-2.292,-0.844", Rotation="0,0,0", Scale="1.000,1.000,1.186") {
    scaleMode=BMM_Additive
    translationMode=BMM_Additive
    translationSpace=BCS_ParentBoneSpace
    alphaBoolBlend {blendOption=Linear}
    boneToModify {boneName=pelvis_side_l}
    scale {z=1.186}
    translation=<-0.228 -2.292 -0.844>
   }
   n15 = Get thigh_rot_r()
   n16 = "Transform (Modify) Bone - Bone: pelvis_r"(ComponentPose=@n14, Alpha=@n15, Translation="3.420,-0.188,-0.379", Rotation="0,0,0", Scale="1.2,1,1") {
    scaleMode=BMM_Additive
    translationMode=BMM_Additive
    translationSpace=BCS_ParentBoneSpace
    alphaBoolBlend {blendOption=Linear}
    boneToModify {boneName=pelvis_r}
    scale {x=1.2}
    translation=<3.42 -0.188 -0.379>
   }
   n17 = "Transform (Modify) Bone - Bone: pelvis_mid_r"(ComponentPose=@n16, Alpha=@n15, Translation="-0.733,-1.466,-0.699", Rotation="0,0,0", Scale="1.204,0.907,1.000") {
    scaleMode=BMM_Additive
    translationMode=BMM_Additive
    translationSpace=BCS_ParentBoneSpace
    alphaBoolBlend {blendOption=Linear}
    boneToModify {boneName=pelvis_mid_r}
    scale=<1.204 0.907>
    translation=<-0.733 -1.466 -0.699>
   }
   n18 = "Transform (Modify) Bone - Bone: pelvis_side_r"(ComponentPose=@n17, Alpha=@n15, Translation="-0.557,0.579,-0.640", Rotation="0,0,0", Scale="1,1,1") {
    scaleMode=BMM_Additive
    translationMode=BMM_Additive
    translationSpace=BCS_ParentBoneSpace
    alphaBoolBlend {blendOption=Linear}
    boneToModify {boneName=pelvis_side_r}
    translation=<-0.557 0.579 -0.64>
   }
   n19 = "Transform (Modify) Bone - Bone: dyn_belt_main"!(ComponentPose=@n18, bAlphaBoolEnabled:bool=Get bIsCrouching(), Translation="-4,0.000,0.000", Rotation="0.000,-2,0.000", Scale="1,1,1") -> (Pose) {
    alphaInputType=Bool
    rotationMode=BMM_Additive
    rotationSpace=BCS_BoneSpace
    translationMode=BMM_Additive
    translationSpace=BCS_BoneSpace
    alphaBoolBlend {blendOption=Linear}
    alphaScaleBiasClamp {bInterpResult=true}
    boneToModify {boneName=dyn_belt_main}
    rotation {yaw=-2}
    translation {x=-4}
   }
   n20 = Get thigh_rotation_l() -> (thigh_rotation_l:double)
   n21 = "Transform (Modify) Bone - Bone: dyn_belt_fr_out_l"(ComponentPose=@n19, Alpha=@n20, Translation="-3,0.000,0.000", Rotation="0,0,0", Scale="1,1,1") {
    rotationMode=BMM_Additive
    rotationSpace=BCS_BoneSpace
    translationMode=BMM_Additive
    translationSpace=BCS_BoneSpace
    alphaBoolBlend {blendOption=Linear}
    alphaScaleBiasClamp {bInterpResult=true}
    boneToModify {boneName=dyn_belt_fr_out_l}
    translation {x=-3}
   }
   n22 = "Transform (Modify) Bone - Bone: dyn_belt_fr_mid_l"(ComponentPose=@n21, Alpha=@n20, Translation="-4,0.000,-0.000", Rotation="0,0,0", Scale="1,1,1") {
    rotationMode=BMM_Additive
    rotationSpace=BCS_BoneSpace
    translationMode=BMM_Additive
    translationSpace=BCS_BoneSpace
    alphaBoolBlend {blendOption=Linear}
    alphaScaleBiasClamp {bInterpResult=true}
    boneToModify {boneName=dyn_belt_fr_mid_l}
    translation {x=-4}
   }
   n23 = "Transform (Modify) Bone - Bone: dyn_belt_fr_l"(ComponentPose=@n22, Alpha=@n20, Translation="-1,-0.000,-0.000", Rotation="0.000,2,0.000", Scale="1,1,1") {
    rotationMode=BMM_Additive
    rotationSpace=BCS_BoneSpace
    translationMode=BMM_Additive
    translationSpace=BCS_BoneSpace
    alphaBoolBlend {blendOption=Linear}
    alphaScaleBiasClamp {bInterpResult=true}
    boneToModify {boneName=dyn_belt_fr_l}
    rotation {yaw=2}
    translation {x=-1}
   }
   n24 = "Transform (Modify) Bone - Bone: dyn_belt_fr_r"(ComponentPose=@n23, Alpha=@n20, Translation="-0.706,-0.261,-0.000", Rotation="0,0,0", Scale="1,1,1") {
    rotationMode=BMM_Additive
    rotationSpace=BCS_BoneSpace
    translationMode=BMM_Additive
    translationSpace=BCS_BoneSpace
    alphaBoolBlend {blendOption=Linear}
    alphaScaleBiasClamp {bInterpResult=true}
    boneToModify {boneName=dyn_belt_fr_r}
    translation=<-0.706 -0.261>
   }
   n25 = "Transform (Modify) Bone - Bone: dyn_belt_fr_02_l"(ComponentPose=@n24, Alpha=@n20, Translation="-2.000,-1.75,0.000", Rotation="0,0,0", Scale="1,1,1") {
    rotationMode=BMM_Additive
    rotationSpace=BCS_BoneSpace
    translationMode=BMM_Additive
    translationSpace=BCS_BoneSpace
    alphaBoolBlend {blendOption=Linear}
    alphaScaleBiasClamp {bInterpResult=true}
    boneToModify {boneName=dyn_belt_fr_02_l}
    translation=<-2 -1.75>
   }
   n26 = "Transform (Modify) Bone - Bone: dyn_belt_fr_out_r"(ComponentPose=@n25, Alpha=@n1, Translation="-4.2,1.575,0.205", Rotation="8.646,2.435,15.801", Scale="1,1,1") {
    rotationMode=BMM_Additive
    rotationSpace=BCS_BoneSpace
    translationMode=BMM_Additive
    translationSpace=BCS_BoneSpace
    alphaBoolBlend {blendOption=Linear}
    alphaScaleBiasClamp {bInterpResult=true, interpSpeedDecreasing=15, interpSpeedIncreasing=15}
    boneToModify {boneName=dyn_belt_fr_out_r}
    rotation=<rot 8.646 2.435 15.801>
    translation=<-4.2 1.575 0.205>
   }
   n27 = "Transform (Modify) Bone - Bone: dyn_belt_fr_mid_r"(ComponentPose=@n26, Alpha=@n1, Translation="-3.798,1.325,0.202", Rotation="-2.748,0.000,0", Scale="1,1,1") {
    rotationMode=BMM_Additive
    rotationSpace=BCS_BoneSpace
    translationMode=BMM_Additive
    translationSpace=BCS_BoneSpace
    alphaBoolBlend {blendOption=Linear}
    alphaScaleBiasClamp {bInterpResult=true}
    boneToModify {boneName=dyn_belt_fr_mid_r}
    rotation {pitch=-2.748}
    translation=<-3.798 1.325 0.202>
   }
   n28 = "Transform (Modify) Bone - Bone: dyn_belt_fr_r"(ComponentPose=@n27, Alpha=@n1, Translation="-2.152,1.117,1.007", Rotation="0.000,-8.912,0.000", Scale="1,1,1") {
    rotationMode=BMM_Additive
    rotationSpace=BCS_BoneSpace
    translationMode=BMM_Additive
    translationSpace=BCS_BoneSpace
    alphaBoolBlend {blendOption=Linear}
    alphaScaleBiasClamp {bInterpResult=true}
    boneToModify {boneName=dyn_belt_fr_r}
    rotation {yaw=-8.912}
    translation=<-2.152 1.117 1.007>
   }
   n29 = "Transform (Modify) Bone - Bone: dyn_belt_fr_r"(ComponentPose=@n28, Alpha=@n1, Translation="-0.706,0.261,-0.000", Rotation="0.000,-5.041,0.000", Scale="1,1,1") {
    rotationMode=BMM_Additive
    rotationSpace=BCS_BoneSpace
    translationMode=BMM_Additive
    translationSpace=BCS_BoneSpace
    alphaBoolBlend {blendOption=Linear}
    alphaScaleBiasClamp {bInterpResult=true}
    boneToModify {boneName=dyn_belt_fr_r}
    rotation {yaw=-5.041}
    translation=<-0.706 0.261>
   }
   n30 = "Transform (Modify) Bone - Bone: dyn_belt_main"!(ComponentPose=@n29, Alpha:float=@n1, Translation="-0.478,0.000,0.000", Rotation="0,0,0", Scale="1,1,1") -> (Pose) {
    rotationMode=BMM_Additive
    rotationSpace=BCS_BoneSpace
    translationMode=BMM_Additive
    translationSpace=BCS_BoneSpace
    alphaBoolBlend {blendOption=Linear}
    alphaScaleBiasClamp {bInterpResult=true}
    boneToModify {boneName=dyn_belt_main}
    translation {x=-0.4782}
   }
   n31 = "Transform (Modify) Bone - Bone: dyn_cod_piece_fr"!(
    ComponentPose=@n30
    bAlphaBoolEnabled:bool=Get crouchMoving() -> (crouchMoving:bool)
    Translation="-2,0.000,0.000"
    Rotation="0,0,0"
    Scale="1,1,1"
   ) -> (Pose) {
    alphaInputType=Bool
    rotationMode=BMM_Additive
    rotationSpace=BCS_BoneSpace
    translationMode=BMM_Additive
    translationSpace=BCS_BoneSpace
    alphaBoolBlend {blendInTime=0.1, blendOption=Linear, blendOutTime=0.1}
    alphaScaleBiasClamp {bInterpResult=true}
    boneToModify {boneName=dyn_cod_piece_fr}
    translation {x=-2}
   }
   n32 = "Constraint - dyn_chest_armor_fr_l_MSTR (PP)"!(ComponentPose=@n31, Alpha=1, ConstraintWeights_0=0.6, ConstraintWeights_1:float=0.4) -> (Pose) {
    alphaBoolBlend {blendOption=Linear}
    boneToModify {boneName=dyn_chest_armor_fr_l_MSTR}
    constraintSetup [{targetBone {boneName=spine_04}}, {targetBone {boneName=spine_03}}]
    constraintWeights [1,1]
   }
   n33 = "Constraint - dyn_chest_armor_fr_r_MSTR (PP)"!(ComponentPose=@n32, Alpha=1, ConstraintWeights_0=0.6, ConstraintWeights_1:float=0.4) -> (Pose) {
    alphaBoolBlend {blendOption=Linear}
    boneToModify {boneName=dyn_chest_armor_fr_r_MSTR}
    constraintSetup [{targetBone {boneName=spine_04}}, {targetBone {boneName=spine_03}}]
    constraintWeights [1,1]
   }
   n34 = "Constraint - dyn_chest_armor_bk_l_MSTR (PP)"!(ComponentPose=@n33, Alpha=1, ConstraintWeights_0=0.6, ConstraintWeights_1:float=0.4) -> (Pose) {
    alphaBoolBlend {blendOption=Linear}
    boneToModify {boneName=dyn_chest_armor_bk_l_MSTR}
    constraintSetup [{targetBone {boneName=spine_04}}, {targetBone {boneName=spine_03}}]
    constraintWeights [1,1]
   }
   n35 = "Constraint - dyn_chest_armor_bk_r_MSTR (PP)"!(ComponentPose=@n34, Alpha=1, ConstraintWeights_0=0.6, ConstraintWeights_1:float=0.4) -> (Pose) {
    alphaBoolBlend {blendOption=Linear}
    boneToModify {boneName=dyn_chest_armor_bk_r_MSTR}
    constraintSetup [{targetBone {boneName=spine_04}}, {targetBone {boneName=spine_03}}]
    constraintWeights [1,1]
   }
   Output Pose(Result=Component To Local(ComponentPose=@n35)) {layerGroup=None}
  ]}
 EventGraph {nodes [
   Event BlueprintUpdateAnimation()
   > n1 = Sequence!(execute) -> (then_0, then_1, then_2:Exec)
   n2 = Sequence!(execute=@n1.then_1) -> (then_0, then_1, then_2, then_3:Exec)
   n3 = BreakTransform(InTransform=GetDeltaTransformFromRefPose(self=GetOwnerSkeletalMeshComponent(), BoneName=clavicle_r, BaseName=spine_05))
   > Set clav_r_rot(clav_r_rot:double=MapRangeClamped(Value=BreakRotator(InRot=@n3.Rotation).Pitch, InRangeA=10, InRangeB=40)) -> (Output_Get:double)
   n4 = BreakTransform(InTransform=GetDeltaTransformFromRefPose(self=GetOwnerSkeletalMeshComponent(), BoneName=clavicle_l, BaseName=spine_05))
   Set clav_l_rot(execute=@n2.then_1, clav_l_rot:double=MapRangeClamped(Value=BreakRotator(InRot=@n4.Rotation).Pitch, InRangeA=10, InRangeB=40)) -> (Output_Get:double)
   n5 = GetOwnerSkeletalMeshComponent()
   n6 = BreakRotator(InRot=BreakTransform(InTransform=GetDeltaTransformFromRefPose(self=@n5, BoneName=thigh_l, BaseName=pelvis)).Rotation)
   Set thigh_rotation_l(execute=@n2.then_3, thigh_rotation_l:double=MapRangeClamped(Value=@n6.Yaw, InRangeA=-50, InRangeB=-90)) -> (Output_Get:double)
   > Set thigh_rot_l(thigh_rot_l=MapRangeClamped(Value=@n6.Yaw, InRangeA=-15, InRangeB=-110))
   n7 = GetOwnerSkeletalMeshComponent()
   n8 = BreakRotator(InRot=BreakTransform(InTransform=GetDeltaTransformFromRefPose(self=@n7, BoneName=thigh_r, BaseName=pelvis)).Rotation)
   Set thigh rotation_r(execute=@n2.then_2, thigh rotation_r:double=MapRangeClamped(Value=@n8.Yaw, InRangeA=-50, InRangeB=-90)) -> (Output_Get:double)
   > Set thigh_rot_r(thigh_rot_r=MapRangeClamped(Value=@n8.Yaw, InRangeA=0, InRangeB=-90))
   Set IsSkyDiving(execute=@n1.then_0, IsSkyDiving:double="To Float (Boolean)"(InBool=Get bIsSkyDiving())) -> (Output_Get:double)
   > Set crouchMoving(crouchMoving:bool=AND Boolean(A=Get bIsCrouching(), B=Get bIsMoving2D())) -> (Output_Get:bool)
  ]}
]
variables [
 clav_l_rot {type=number}
 clav_r_rot {type=number}
 crouchMoving {type=boolean}
 IsSkyDiving {type=number}
 thigh rotation_r {type=number}
 thigh_rot_l {type=number}
 thigh_rot_r {type=number}
 thigh_rotation_l {type=number}
]
